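# Sound cue for a held object's death, playing a metal demolition impact under the Fort_Impacts class.
# Each variant mixes close, medium and far layers with crossfade attenuations, delaying the medium layer slightly.
# A quality switch sends the first two levels to a mix with random waves in every layer, and the last two to one randomizing only the close layer, with pitch-modulated single medium and far waves.
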SoundCue /Game/Sounds/Apollo/HeldObjects/HeldObject_Death_Cue
root=QualityLevel_0
nodes [
 {
  id=Attenuation_10
  children [Modulator_6]
  properties {attenuationSettings=/Game/Sounds/Attenuation/Fort_Demolition_Crossfade_Global_Hit_Close_Att}
 }
 {
  id=Attenuation_3
  children [Modulator_4]
  properties {attenuationSettings=/Game/Sounds/Attenuation/Fort_Demolition_Crossfade_Global_Hit_Far_Att}
 }
 {
  id=Attenuation_4
  children [Delay_1]
  properties {attenuationSettings=/Game/Sounds/Attenuation/Fort_Demolition_Crossfade_Global_Hit_Medium_Att}
 }
 {
  id=Attenuation_5
  children [Modulator_2]
  properties {attenuationSettings=&Fort_Demolition_Crossfade_Global_Hit_Close_Att}
 }
 {
  id=Attenuation_8
  children [Random_0]
  properties {attenuationSettings=&Fort_Demolition_Crossfade_Global_Hit_Far_Att}
 }
 {
  id=Attenuation_9
  children [Delay_6]
  properties {attenuationSettings=&Fort_Demolition_Crossfade_Global_Hit_Medium_Att}
 }
 {id=Delay_1, children [Modulator_3], properties {delayMax=0.2, delayMin=0.1}}
 {id=Delay_6, children [Modulator_11], properties {delayMax=0.2, delayMin=0.1}}
 {id=Mixer_1, children [Attenuation_5, Attenuation_4, Attenuation_3], properties {inputVolume [1,1,1]}}
 {id=Mixer_7, children [Attenuation_10, Attenuation_9, Attenuation_8], properties {inputVolume [1,1,1]}}
 {id=Modulator_11, children [Random_18]}
 {id=Modulator_2, children [Random_4]}
 {id=Modulator_3, children [WavePlayer_14], properties {pitchMax=1.07, pitchMin=0.92}}
 {id=Modulator_4, children [WavePlayer_16], properties {pitchMax=1.1, pitchMin=0.9}}
 {id=Modulator_6, children [Random_15]}
 {id=QualityLevel_0, children [Mixer_7,Mixer_7,Mixer_1,Mixer_1]}
 {id=Random_0, children [WavePlayer_0,WavePlayer_1], properties {weights [1,1]}}
 {id=Random_15, children [WavePlayer_43, WavePlayer_44], properties {weights [1,1]}}
 {id=Random_18, children [WavePlayer_46, WavePlayer_47], properties {weights [1,1]}}
 {id=Random_4, children [WavePlayer_12, WavePlayer_13], properties {bRandomizeWithoutReplacement=false, weights [1,1]}}
 {id=WavePlayer_0, properties {soundWaveAssetPtr=/Game/Sounds/Fort_Impact_Sounds/Demolition/Defaults/Metal_Distant_Impact_01}}
 {id=WavePlayer_1, properties {soundWaveAssetPtr=/Game/Sounds/Fort_Impact_Sounds/Demolition/Defaults/Metal_Distant_Impact_03}}
 {id=WavePlayer_12, properties {soundWaveAssetPtr=/Game/Sounds/Fort_Impact_Sounds/Demolition/Defaults/final_metal_new_01}}
 {id=WavePlayer_13, properties {soundWaveAssetPtr=/Game/Sounds/Fort_Impact_Sounds/Demolition/Defaults/final_metal_new_02}}
 {id=WavePlayer_14, properties {soundWaveAssetPtr=/Game/Sounds/Fort_Impact_Sounds/Demolition/Defaults/metal_close_impact_01}}
 {id=WavePlayer_16, properties {soundWaveAssetPtr=&Metal_Distant_Impact_01}}
 {id=WavePlayer_43, properties {soundWaveAssetPtr=&final_metal_new_01}}
 {id=WavePlayer_44, properties {soundWaveAssetPtr=&final_metal_new_02}}
 {id=WavePlayer_46, properties {soundWaveAssetPtr=&metal_close_impact_01}}
 {id=WavePlayer_47, properties {soundWaveAssetPtr=/Game/Sounds/Fort_Impact_Sounds/Demolition/Defaults/metal_close_impact_02}}
]
properties {
 attenuationSettings=/Game/Sounds/Attenuation/Fort_Demolition_Destroy_Medium_Sized_Object_Att
 maxDistance=8000
 soundClassObject=&Fort_Impacts
 volumeMultiplier=0.885
 concurrencyOverrides {maxCount=1}
 concurrencySet [/Game/Sounds/Concurrency_Rules/FourMaximum_StopFurthestOldest_VolumeScale]
}
