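# Sound cue for the foley heard when picking up a held object, built from player finger-wave foley waves.
# A mixer layers two random picks, one from each of two three-variant sets, and a modulator adds slight random volume variation on the result.
# It plays in the player foley sound class with footstep attenuation, and its own concurrency limit stops the oldest instance when exceeded.
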SoundCue /Game/Sounds/Apollo/HeldObjects/HeldObject_Pickup_Foley_Cue
root=Modulator_12
nodes [
 {id=Mixer_0, children [Random_1,Random_0], properties {inputVolume [1,1]}}
 {id=Modulator_12, children [Mixer_0], properties {pitchMax=1, volumeMax=1, volumeMin=0.9}}
 {id=Random_0, children [WavePlayer_0,WavePlayer_1,WavePlayer_2], properties {weights [1,1,1]}}
 {id=Random_1, children [WavePlayer_4,WavePlayer_5,WavePlayer_6], properties {weights [1,1,1]}}
 {id=WavePlayer_0, properties {soundWaveAssetPtr=/Game/Sounds/Fort_Foley_Sounds/Moves/PlayerFoley_FingerWave_01}}
 {id=WavePlayer_1, properties {soundWaveAssetPtr=/Game/Sounds/Fort_Foley_Sounds/Moves/PlayerFoley_FingerWave_02}}
 {id=WavePlayer_2, properties {soundWaveAssetPtr=/Game/Sounds/Fort_Foley_Sounds/Moves/PlayerFoley_FingerWave_05}}
 {id=WavePlayer_4, properties {soundWaveAssetPtr=/Game/Sounds/Fort_Foley_Sounds/Moves/PlayerFoley_FingerWave_B_03}}
 {id=WavePlayer_5, properties {soundWaveAssetPtr=/Game/Sounds/Fort_Foley_Sounds/Moves/PlayerFoley_FingerWave_B_01}}
 {id=WavePlayer_6, properties {soundWaveAssetPtr=/Game/Sounds/Fort_Foley_Sounds/Moves/PlayerFoley_FingerWave_B_02}}
]
properties {
 attenuationSettings=/Game/Sounds/Attenuation/Fort_Player_Footsteps_Att
 bOverrideConcurrency=true
 maxDistance=3500
 priority=0.5
 soundClassObject=&Fort_Player_Foley
 volumeMultiplier=0.8
 concurrencyOverrides {maxCount=4, resolutionRule=StopOldest}
}
