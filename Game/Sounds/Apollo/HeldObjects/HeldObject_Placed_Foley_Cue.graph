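# Sound cue for the foley heard when a held object is placed, built from reused player turn-movement sounds.
# A Random node picks one of three equally weighted turn foley waves, and a Modulator varies its pitch slightly below normal.
# It plays in the player foley sound class with footstep attenuation.
# A concurrency override caps simultaneous instances and stops the oldest.
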SoundCue /Game/Sounds/Apollo/HeldObjects/HeldObject_Placed_Foley_Cue
root=Modulator_12
nodes [
 {id=Modulator_12, children [Random_0], properties {pitchMax=0.95, pitchMin=0.9}}
 {id=Random_0, children [WavePlayer_3,WavePlayer_1,WavePlayer_2], properties {weights [1,1,1]}}
 {id=WavePlayer_1, properties {soundWaveAssetPtr=/Game/Sounds/Fort_Foley_Sounds/Moves/PlayerFoley_Turn_01}}
 {id=WavePlayer_2, properties {soundWaveAssetPtr=/Game/Sounds/Fort_Foley_Sounds/Moves/PlayerFoley_Turn_02}}
 {id=WavePlayer_3, properties {soundWaveAssetPtr=/Game/Sounds/Fort_Foley_Sounds/Moves/PlayerFoley_Turn_03}}
]
properties {
 attenuationSettings=/Game/Sounds/Attenuation/Fort_Player_Footsteps_Att
 bOverrideConcurrency=true
 maxDistance=3500
 priority=0.5
 soundClassObject=&Fort_Player_Foley
 volumeMultiplier=0.8
 concurrencyOverrides {maxCount=4, resolutionRule=StopOldest}
}
