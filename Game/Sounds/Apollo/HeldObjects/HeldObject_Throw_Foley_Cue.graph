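# Sound cue that plays a cloth-and-movement foley sound for throwing a held object, going by its name and folder.
# A random node picks evenly among three reused player-run foley waves, and a modulator slightly varies the volume.
# It plays in the player foley sound class with footstep attenuation and overrides concurrency to cap simultaneous instances, stopping the oldest.
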SoundCue /Game/Sounds/Apollo/HeldObjects/HeldObject_Throw_Foley_Cue
root=Modulator_12
nodes [
 {id=Modulator_12, children [Random_3], properties {pitchMax=1, volumeMax=1, volumeMin=0.9}}
 {id=Random_3, children [WavePlayer_14, WavePlayer_12, WavePlayer_17], properties {weights [1,1,1]}}
 {id=WavePlayer_12, properties {soundWaveAssetPtr=/Game/Sounds/Fort_Foley_Sounds/Moves/player_run_foley_09}}
 {id=WavePlayer_14, properties {soundWaveAssetPtr=/Game/Sounds/Fort_Foley_Sounds/Moves/player_run_foley_05}}
 {id=WavePlayer_17, properties {soundWaveAssetPtr=/Game/Sounds/Fort_Foley_Sounds/Moves/player_run_foley_10}}
]
properties {
 attenuationSettings=/Game/Sounds/Attenuation/Fort_Player_Footsteps_Att
 bOverrideConcurrency=true
 maxDistance=3500
 priority=0.5
 soundClassObject=&Fort_Player_Foley
 volumeMultiplier=0.65
 concurrencyOverrides {maxCount=4, resolutionRule=StopOldest}
}
